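# Stereo MetaSound jam effect that applies a MIDI-clock-synced stutter to live input, recapturing audio while active.
# IsActive fires a quantized start that picks a random subdivision for StutterController, which drives per-channel SimpleSamplers.
# A gated subdivision trigger forces periodic recapture until deactivation.
# Samplers feed biquad filters swept by a tempo-based LFO at a random power-of-two rate, smoothly crossfaded against dry input; Rate is unconnected.
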MetaSoundSource /FMJamSystem/Audio/JamEffects/MSS_Jam_Effect_StutteRecapture
inputs [
 Input L {type=Audio}
 Input R {type=Audio}
 IsActive {type=Bool}
 MIDI Clock {type=MIDIClock}
 Rate {type=Float}
 UE.Source.OnPlay {type=Trigger}
]
interfaces [UE.OutputFormat.Stereo:1.0, UE.Source:1.1]
outputs [
 "UE.OutputFormat.Stereo.Audio:0" {type=Audio}
 "UE.OutputFormat.Stereo.Audio:1" {type=Audio}
]
pages [Default {
  input_defaults {Input L=null, Input R=null, IsActive=null, MIDI Clock=null, Rate=3, UE.Source.OnPlay=null}
  nodes [
   n1 = UE.InterpTo:Audio(Target=Convert.Float:Bool(Bool=^IsActive).Float) {also {defaults {Interp Time=0.05}}}
   n2 = UE.Reroute(Value=^Input L)
   n3 = UE.Reroute(Value=^MIDI Clock)
   n4 = UE.Reroute(Value=^IsActive)
   TriggerOnValueChange_0 = UE.Trigger On Value Change:Int32(Value=Convert.Int32:Bool(Bool=@n4.Value).Int32)
   n5 = TriggerCompare.Clamp:Bool(A=@n4.Value, Compare=@TriggerOnValueChange_0.Trigger) {also {defaults {B="<MetaSoundBuilderResult.SUCCEEDED: 0>"}}}
   n6 = HarmonixNodes.MidiQuantizeTriggerNode(MIDI Clock=@n3.Value, Trigger Input=@n5.True) {also {defaults {Subdivision=5, Tolerance=0}, version [0,1]}}
   n7 = UE.Reroute(Value=@n6.Trigger Out)
   n8 = VariableAccessor.Int32:Array(Variable=InitVariable.Int32:Array().Variable)
   n9 = UE.RandomInt32(
    Next=@n7.Value
    Max=UE.Subtract:Int32(PrimaryOperand=Array.Num:Int32:Array(Array=@n8.Value).Num) {also {defaults {AdditionalOperands=1}}}.Out
   ) {also {version [1,1]}}
   n10 = UE.Reroute(Value=@n8.Value)
   n11 = Convert.Enum:SubdivisionQuantizationType:Int32(Int32=Array.Get:Int32:Array(Trigger=@n9.On Next, Index=@n9.Value, Array=UE.Reroute(Value=@n10.Value).Value).Element)
   n12 = UE.Reroute(Value=UE.Reroute(Value=@n5.False).Value)
   n13 = UE.Reroute(Value=@n3.Value)
   n14 = HarmonixNodes.StutterController(
    Audible Duration=@n11."Enum:SubdivisionQuantizationType"
    Spacing=@n11."Enum:SubdivisionQuantizationType"
    Start=UE.Reroute(Value=UE.Reroute(Value=@n7.Value).Value).Value
    Stop=@n12.Value
    MIDI Clock=@n13.Value
   ) {also {defaults {Capture Duration=5}, version [0,1]}}
   n15 = UE.Reroute(Value=@n14.Play Duration Sec)
   TriggerControl_0 = UE.Trigger Control(
    Trigger In=HarmonixNodes.MidiClockSubdivisionTrigger(MIDI Clock=@n13.Value) {also {defaults {Base Grid Size=6}}}.Trigger Out
    Close=@n12.Value
    Open=@n7.Value
   )
   TriggerAnyTriggerAny2_0 = "TriggerAny.Trigger Any (2)"(In 1=@n14.Capture, In 0=UE.Reroute(Value=@TriggerControl_0.Trigger Out).Value)
   n16 = UE.Reroute(Value=@TriggerAnyTriggerAny2_0.Out)
   n17 = UE.Reroute(Value=@n14.Reset)
   n18 = UE.Reroute(Value=@n14.Capture Duration Sec)
   TriggerDelay_0 = UE.Trigger Delay(In=@n14.Play) {also {defaults {Delay Time=0.01}}}
   n19 = UE.Reroute(Value=@TriggerDelay_0.Out)
   n20 = UE.Divide:Float(PrimaryOperand=HarmonixNodes.MidiClockState(MIDI Clock=^MIDI Clock) {also {version [0,1]}}.Tempo) {also {defaults {AdditionalOperands=60}}}
   n21 = VariableAccessor.Trigger(Variable=VariableMutator.Trigger(Value=@n6.Trigger Out, Variable=InitVariable.Trigger().Variable).Variable)
   n22 = UE.Power:Float(AdditionalOperands=Convert.Float:Int32(Int32=UE.RandomInt32(Next=@n21.Value) {also {defaults {Max=4}, version [1,1]}}.Value).Float) {also {defaults {PrimaryOperand=2}}}
   LinearToLogFrequency_0 = UE.Linear To Log Frequency(Value=UE.LFO:Audio(Frequency=UE.Divide:Float(PrimaryOperand=@n20.Out, AdditionalOperands=@n22.Out).Out, Sync=@n21.Value) {also {defaults {Min Value=0}}}.Out) {also {defaults {Max Range=300, Min Range=5000}}}
   n23 = UE.Reroute(Value=UE.InterpTo:Audio(Target=@LinearToLogFrequency_0.Frequency) {also {defaults {Interp Time=0.05}}}.Value)
   BiquadFilter_0 = UE.Biquad Filter:Audio(
    In=HarmonixNodes.SimpleSampler(In 0=@n2.Value, Play Duration Sec=@n15.Value, Capture=@n16.Value, Reset=@n17.Value, Capture Duration Sec=@n18.Value, Play=@n19.Value) {also {defaults {Enable="<MetaSoundBuilderResult.SUCCEEDED: 0>"}, version [0,1]}}.Out 0
    Cutoff Frequency=@n23.Value
   ) {also {defaults {Bandwidth=4, Gain=13, Type=2}}}
   CrossfadeTriggerRouteAudio2_0 = "Crossfade.Trigger Route (Audio, 2):Audio"(
    Crossfade Value=@n1.Value
    In 1=UE.Multiply:Audio by Float(PrimaryOperand=@BiquadFilter_0.Out) {also {defaults {AdditionalOperands=1.2}}}.Out
    In 0=@n2.Value
   )
   n24 = UE.Reroute(Value=^Input R)
   BiquadFilter_1 = UE.Biquad Filter:Audio(
    In=HarmonixNodes.SimpleSampler(In 0=@n24.Value, Play Duration Sec=@n15.Value, Capture=@n16.Value, Reset=@n17.Value, Capture Duration Sec=@n18.Value, Play=@n19.Value) {also {defaults {Enable="<MetaSoundBuilderResult.SUCCEEDED: 0>"}, version [0,1]}}.Out 0
    Cutoff Frequency=@n23.Value
   ) {also {defaults {Bandwidth=4, Gain=13, Type=2}}}
   CrossfadeTriggerRouteAudio2_1 = "Crossfade.Trigger Route (Audio, 2):Audio"(
    Crossfade Value=@n1.Value
    In 1=UE.Multiply:Audio by Float(PrimaryOperand=@BiquadFilter_1.Out) {also {defaults {AdditionalOperands=1.2}}}.Out
    In 0=@n24.Value
   )
   InitVariable.StutterSequenceAsset:Array()
  ]
  links [
   {dst="@UE.OutputFormat.Stereo.Audio:0", src=CrossfadeTriggerRouteAudio2_0.Out}
   {dst="@UE.OutputFormat.Stereo.Audio:1", src=CrossfadeTriggerRouteAudio2_1.Out}
  ]
 }]
properties {
 assetImportData=&self:AssetImportData
 bLooping=true
 duration=10000
 numChannels=2
 outputFormat=Stereo
 sampleRate=48000
}
