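# MetaSound stereo jam effect that replaces Input L/R with a tempo-synced, filtered stutter while IsActive is on.
# Activation is MIDI-clock quantized, picks a random subdivision, and drives a StutterController feeding per-channel SimpleSamplers.
# Captured audio runs through biquad filters swept by a beat-synced LFO with randomized period, then crossfades against dry over an interpolated IsActive.
# Rate, Fabric AudioIn/AudioOut and a gated subdivision trigger are left unconnected.
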MetaSoundSource /FMJamSystem/Audio/JamEffects/MSS_Jam_Effect_StutterFiltered
inputs [
 Fabric.Audio.AudioIn {type=Audio}
 Fabric.Common.MidiClockIn {type=MIDIClock}
 Fabric.Common.RootNote {type=Int32}
 Fabric.Common.Scale {type=Enum:MusicalScale}
 Fabric.Common.TransportIn {type=MusicTransport}
 Input L {type=Audio}
 Input R {type=Audio}
 IsActive {type=Bool}
 Rate {type=Float}
 UE.Source.OnPlay {type=Trigger}
]
interfaces [Fabric.AudioEffect:1.0, UE.OutputFormat.Stereo:1.0, UE.Source:1.1]
outputs [
 Fabric.AudioEffect.AudioOut {type=Audio}
 "UE.OutputFormat.Stereo.Audio:0" {type=Audio}
 "UE.OutputFormat.Stereo.Audio:1" {type=Audio}
]
pages [Default {
  input_defaults {
   Fabric.Audio.AudioIn=null
   Fabric.Common.MidiClockIn=null
   Fabric.Common.RootNote=0
   Fabric.Common.Scale=0
   Fabric.Common.TransportIn=null
   Input L=null
   Input R=null
   IsActive=null
   Rate=3
   UE.Source.OnPlay=null
  }
  nodes [
   n1 = UE.InterpTo:Audio(Target=Convert.Float:Bool(Bool=^IsActive).Float) {also {defaults {Interp Time=0.05}}}
   n2 = UE.Reroute(Value=^Fabric.Common.MidiClockIn)
   n3 = UE.Reroute(Value=^IsActive)
   TriggerOnValueChange_0 = UE.Trigger On Value Change:Int32(Value=Convert.Int32:Bool(Bool=@n3.Value).Int32)
   n4 = TriggerCompare.Clamp:Bool(A=@n3.Value, Compare=@TriggerOnValueChange_0.Trigger) {also {defaults {B="<MetaSoundBuilderResult.SUCCEEDED: 0>"}}}
   n5 = HarmonixNodes.MidiQuantizeTriggerNode(MIDI Clock=@n2.Value, Trigger Input=@n4.True) {also {defaults {Subdivision=5, Tolerance=0}, version [0,1]}}
   n6 = UE.Reroute(Value=@n5.Trigger Out)
   n7 = VariableAccessor.Int32:Array(Variable=InitVariable.Int32:Array().Variable)
   n8 = UE.RandomInt32(
    Next=@n6.Value
    Max=UE.Subtract:Int32(PrimaryOperand=Array.Num:Int32:Array(Array=@n7.Value).Num) {also {defaults {AdditionalOperands=1}}}.Out
   ) {also {version [1,1]}}
   n9 = UE.Reroute(Value=@n7.Value)
   n10 = Convert.Enum:SubdivisionQuantizationType:Int32(Int32=Array.Get:Int32:Array(Trigger=@n8.On Next, Index=@n8.Value, Array=UE.Reroute(Value=@n9.Value).Value).Element)
   n11 = UE.Reroute(Value=UE.Reroute(Value=@n4.False).Value)
   n12 = UE.Reroute(Value=@n2.Value)
   n13 = HarmonixNodes.StutterController(
    Audible Duration=@n10."Enum:SubdivisionQuantizationType"
    Spacing=@n10."Enum:SubdivisionQuantizationType"
    Start=UE.Reroute(Value=UE.Reroute(Value=@n6.Value).Value).Value
    Stop=@n11.Value
    MIDI Clock=@n12.Value
   ) {also {defaults {Capture Duration=6}, version [0,1]}}
   n14 = UE.Reroute(Value=^Input L)
   TriggerAnyTriggerAny2_0 = "TriggerAny.Trigger Any (2)"(In 1=@n13.Capture)
   n15 = VariableMutator.Trigger(Value=@n5.Trigger Out, Variable=InitVariable.Trigger().Variable)
   n16 = Convert.Float:Int32(Int32=UE.RandomInt32(Next=VariableAccessor.Trigger(Variable=@n15.Variable).Value) {also {defaults {Max=4}, version [1,1]}}.Value)
   LinearToLogFrequency_0 = UE.Linear To Log Frequency(Value=FabricNodes.FabricLFONode(
     MIDI Clock=^Fabric.Common.MidiClockIn
     Beat Sync Duration=UE.Power:Float(AdditionalOperands=@n16.Float) {also {defaults {PrimaryOperand=2}}}.Out
    ) {also {defaults {Shape=0, Wave Shape=1}, version [0,1]}}.Output) {also {defaults {Max Range=300, Min Range=5000}}}
   n17 = UE.Reroute(Value=UE.InterpTo:Audio(Target=@LinearToLogFrequency_0.Frequency) {also {defaults {Interp Time=0.05}}}.Value)
   BiquadFilter_0 = UE.Biquad Filter:Audio(
    In=HarmonixNodes.SimpleSampler(
     Capture Duration Sec=@n13.Capture Duration Sec
     Play=@n13.Play
     Play Duration Sec=@n13.Play Duration Sec
     Reset=@n13.Reset
     In 0=@n14.Value
     Capture=@TriggerAnyTriggerAny2_0.Out
    ) {also {defaults {Enable="<MetaSoundBuilderResult.SUCCEEDED: 0>"}, version [0,1]}}.Out 0
    Cutoff Frequency=@n17.Value
   ) {also {defaults {Bandwidth=4, Gain=13, Type=2}}}
   CrossfadeTriggerRouteAudio2_0 = "Crossfade.Trigger Route (Audio, 2):Audio"(
    Crossfade Value=@n1.Value
    In 1=UE.Multiply:Audio by Float(PrimaryOperand=@BiquadFilter_0.Out) {also {defaults {AdditionalOperands=1.2}}}.Out
    In 0=@n14.Value
   )
   n18 = UE.Reroute(Value=^Input R)
   BiquadFilter_1 = UE.Biquad Filter:Audio(
    In=HarmonixNodes.SimpleSampler(
     Capture Duration Sec=@n13.Capture Duration Sec
     Play=@n13.Play
     Play Duration Sec=@n13.Play Duration Sec
     Reset=@n13.Reset
     In 0=@n18.Value
     Capture=@TriggerAnyTriggerAny2_0.Out
    ) {also {defaults {Enable="<MetaSoundBuilderResult.SUCCEEDED: 0>"}, version [0,1]}}.Out 0
    Cutoff Frequency=@n17.Value
   ) {also {defaults {Bandwidth=4, Gain=13, Type=2}}}
   CrossfadeTriggerRouteAudio2_1 = "Crossfade.Trigger Route (Audio, 2):Audio"(
    Crossfade Value=@n1.Value
    In 1=UE.Multiply:Audio by Float(PrimaryOperand=@BiquadFilter_1.Out) {also {defaults {AdditionalOperands=1.2}}}.Out
    In 0=@n18.Value
   )
   InitVariable.StutterSequenceAsset:Array()
   TriggerControl_0 = UE.Trigger Control(
    Trigger In=HarmonixNodes.MidiClockSubdivisionTrigger(MIDI Clock=@n12.Value) {also {defaults {Base Grid Size=6}}}.Trigger Out
    Close=@n11.Value
    Open=@n6.Value
   )
   UE.Reroute(Value=@TriggerControl_0.Trigger Out)
  ]
  links [
   {dst="@UE.OutputFormat.Stereo.Audio:0", src=CrossfadeTriggerRouteAudio2_0.Out}
   {dst="@UE.OutputFormat.Stereo.Audio:1", src=CrossfadeTriggerRouteAudio2_1.Out}
  ]
 }]
properties {
 assetImportData=&self:AssetImportData
 bLooping=true
 duration=10000
 numChannels=2
 outputFormat=Stereo
 sampleRate=48000
}
