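# Stereo MetaSound source that appears to be a test bed for FMJamSystem stutter effects on an Around the World loop.
# On play, a Harmonix metronome at the loop's BPM drives a MIDI-clock subdivision trigger that restarts the wave player in time.
# Audio, MIDI clock and the bool-converted IsActive float feed both StutteRecapture and StutterFiltered.
# Only StutteRecapture reaches the output; StutterFiltered's output and OnFinished are unwired.
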MetaSoundSource /FMJamSystem/Audio/JamEffects/FusionFXTest
inputs [
 IsActive {type=Float}
 UE.Source.OnPlay {type=Trigger}
]
interfaces [UE.OutputFormat.Stereo:1.0, UE.Source.OneShot:1.0, UE.Source:1.1]
outputs [
 "UE.OutputFormat.Stereo.Audio:0" {type=Audio}
 "UE.OutputFormat.Stereo.Audio:1" {type=Audio}
 UE.Source.OneShot.OnFinished {type=Trigger}
]
pages [Default {
  input_defaults {IsActive=0, UE.Source.OnPlay=null}
  nodes [
   n1 = Convert.Bool:Float(Float=^IsActive)
   n2 = HarmonixNodes.Metronome(Transport=HarmonixNodes.TriggerToTransport(Play=^UE.Source.OnPlay).Transport) {also {defaults {Pre-roll Bars=0, Simple Loop="<MetaSoundBuilderResult.SUCCEEDED: 0>", Tempo=121.28}}}
   n3 = UE.Reroute(Value=UE.Reroute(Value=@n2.MIDI Clock).Value)
   WavePlayer_0 = UE.Wave Player:Stereo(Play=HarmonixNodes.MidiClockSubdivisionTrigger(MIDI Clock=@n2.MIDI Clock) {also {defaults {Base Grid Size=0, Grid Size Multiplier=4}}}.Trigger Out) {also {defaults {Wave Asset=/FMJamSystem/Audio/AroundTheWorld_Loop_121-28bpm}}}
   MSS_Jam_Effect_StutteRecapture_0 = &/MSS_Jam_Effect_StutteRecapture(IsActive=@n1.Bool, Fabric.Common.MidiClockIn=@n3.Value, Input L=@WavePlayer_0.Out Left, Input R=@WavePlayer_0.Out Right)
   &/MSS_Jam_Effect_StutterFiltered(IsActive=@n1.Bool, Fabric.Common.MidiClockIn=@n3.Value, Input L=@WavePlayer_0.Out Left, Input R=@WavePlayer_0.Out Right)
  ]
  links [
   {dst="@UE.OutputFormat.Stereo.Audio:0", src=MSS_Jam_Effect_StutteRecapture_0.UE.OutputFormat.Stereo.Audio:0}
   {dst="@UE.OutputFormat.Stereo.Audio:1", src=MSS_Jam_Effect_StutteRecapture_0.UE.OutputFormat.Stereo.Audio:1}
  ]
 }]
properties {assetImportData=&self:AssetImportData, numChannels=2, outputFormat=Stereo, sampleRate=48000}
