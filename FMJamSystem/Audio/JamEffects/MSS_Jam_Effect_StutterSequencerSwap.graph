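# Stereo MetaSound source for a music jam stutter effect, playing its own wave player on an internal fixed-tempo metronome.
# GO crossfades the dry left channel with a Simple Sampler that captures and repeats clock-quantized slices, sized by a Rate-indexed divisor.
# A parallel StutterSequencer, swapping Rate-indexed sequence assets on quantized triggers, drives a second sampler whose output is unconnected.
# Input L/R and IsActive are unused; both outputs carry the same signal.
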MetaSoundSource /FMJamSystem/Audio/JamEffects/MSS_Jam_Effect_StutterSequencerSwap
inputs [
 GO {type=Float}
 Input L {type=Audio}
 Input R {type=Audio}
 IsActive {type=Bool}
 Rate {type=Float}
 UE.Source.OnPlay {type=Trigger}
]
interfaces [UE.OutputFormat.Stereo:1.0, UE.Source.OneShot:1.0, UE.Source:1.1]
outputs [
 "UE.OutputFormat.Stereo.Audio:0" {type=Audio}
 "UE.OutputFormat.Stereo.Audio:1" {type=Audio}
 UE.Source.OneShot.OnFinished {type=Trigger}
]
pages [Default {
  input_defaults {GO=0, Input L=null, Input R=null, IsActive=null, Rate=0.008, UE.Source.OnPlay=null}
  nodes [
   FabricNodesFabricMetronome_0 = FabricNodes.Fabric Metronome(Transport=HarmonixNodes.TriggerToTransport(Play=^UE.Source.OnPlay).Transport) {also {defaults {Tempo=121}, version [0,1]}}
   n1 = UE.Reroute(Value=@FabricNodesFabricMetronome_0.MIDI Clock)
   n2 = UE.Reroute(Value=@n1.Value)
   n3 = TriggerCompare.Clamp:Bool(
    A=Convert.Bool:Float(Float=^GO).Bool
    Compare=HarmonixNodes.MidiClockSubdivisionTrigger(MIDI Clock=@n1.Value) {also {defaults {Base Grid Size=3}}}.Trigger Out
   ) {also {defaults {B="<MetaSoundBuilderResult.SUCCEEDED: 0>"}}}
   TriggerOnce_0 = UE.Trigger Once(Reset=@n3.False, Trigger In=@n3.True) {also {defaults {Start Closed="<MetaSoundBuilderResult.SUCCEEDED: 0>"}}}
   n4 = Convert.Int32:Float(Float=^Rate)
   TriggerOnValueChange_1 = UE.Trigger On Value Change:Int32(Value=@n4.Int32)
   n5 = HarmonixNodes.MidiQuantizeTriggerNode(MIDI Clock=@n1.Value, Trigger Input=@TriggerOnValueChange_1.Trigger) {also {defaults {Tolerance=0}, version [0,1]}}
   n6 = VariableAccessor.Float:Array(Variable=InitVariable.Float:Array().Variable)
   n7 = VariableMutator.Float(
    Value=HarmonixNodes.MidiClockState(MIDI Clock=@n1.Value) {also {version [0,1]}}.Secs Per Quarter Note
    Variable=InitVariable.Float().Variable
   )
   n8 = UE.Reroute(Value=UE.Divide:Float(
     AdditionalOperands=Array.Get:Float:Array(Index=@n4.Int32, Trigger=@n5.Trigger Out, Array=@n6.Value).Element
     PrimaryOperand=VariableAccessor.Float(Variable=@n7.Variable).Value
    ).Out)
   WavePlayer_0 = UE.Wave Player:Stereo(Play=HarmonixNodes.MidiClockSubdivisionTrigger(MIDI Clock=@FabricNodesFabricMetronome_0.MIDI Clock) {also {defaults {Base Grid Size=0, Grid Size Multiplier=4}}}.Trigger Out)
   CrossfadeTriggerRouteAudio2_0 = "Crossfade.Trigger Route (Audio, 2):Audio"(
    Crossfade Value=Convert.Float:Bool(Bool=Convert.Bool:Float(Float=^GO).Bool).Float
    In 1=HarmonixNodes.SimpleSampler(
     Capture=HarmonixNodes.MidiQuantizeTriggerNode(MIDI Clock=@n2.Value, Trigger Input=@TriggerOnce_0.Trigger Out) {also {version [0,1]}}.Trigger Out
     Play=UE.Reroute(Value=@n3.True).Value
     Capture Duration Sec=@n8.Value
     Play Duration Sec=@n8.Value
     Reset=@n3.False
     In 0=@WavePlayer_0.Out Left
    ) {also {defaults {Enable="<MetaSoundBuilderResult.SUCCEEDED: 0>"}, version [0,1]}}.Out 0
    In 0=@WavePlayer_0.Out Left
   )
   n9 = Convert.Int32:Float(Float=^Rate)
   n10 = UE.Reroute(Value=UE.Reroute(Value=@FabricNodesFabricMetronome_0.MIDI Clock).Value)
   TriggerOnValueChange_2 = UE.Trigger On Value Change:Int32(Value=@n9.Int32)
   n11 = InitVariable.StutterSequenceAsset:Array()
   n12 = Convert.Bool:Float(Float=^GO)
   TriggerOnValueChange_0 = UE.Trigger On Value Change:Bool(Value=@n12.Bool)
   n13 = &MSP_Branch(
    In=HarmonixNodes.MidiQuantizeTriggerNode(MIDI Clock=UE.Reroute(Value=@n10.Value).Value, Trigger Input=@TriggerOnValueChange_0.Trigger) {also {defaults {Subdivision=4, Tolerance=0}, version [0,1]}}.Trigger Out
    Value=UE.Reroute(Value=UE.Reroute(Value=@n12.Bool).Value).Value
   )
   TriggerRouteTriggerRouteBool2_0 = "TriggerRoute.Trigger Route (Bool, 2):Bool"(Set 1=@n13.False, Set 0=@n13.True) {also {defaults {Value 0="<MetaSoundBuilderResult.SUCCEEDED: 0>"}}}
   n14 = HarmonixNodes.StutterSequencer(
    Stutter Sequence=Array.Get:StutterSequenceAsset:Array(
     Index=@n9.Int32
     Trigger=HarmonixNodes.MidiQuantizeTriggerNode(MIDI Clock=@n10.Value, Trigger Input=@TriggerOnValueChange_2.Trigger) {also {defaults {Subdivision=4, Tolerance=0}, version [0,1]}}.Trigger Out
     Array=VariableAccessor.StutterSequenceAsset:Array(Variable=@n11.Variable).Value
    ).Element
    Stop=@n13.False
    Start=@n13.True
    MIDI Clock=UE.Reroute(Value=UE.Reroute(Value=@n10.Value).Value).Value
    Enable=@TriggerRouteTriggerRouteBool2_0.Value
   ) {also {version [0,1]}}
   HarmonixNodes.SimpleSampler(
    Capture=@n14.Capture
    Capture Duration Sec=@n14.Capture Duration Sec
    Play=@n14.Play
    Play Duration Sec=@n14.Play Duration Sec
    Reset=@n14.Reset
    Reverse Playback=@n14.Reverse Playback
    Enable=@TriggerRouteTriggerRouteBool2_0.Value
    In 0=@WavePlayer_0.Out Left
   ) {also {version [0,1]}}
  ]
  links [
   {dst="@UE.OutputFormat.Stereo.Audio:0", src=CrossfadeTriggerRouteAudio2_0.Out}
   {dst="@UE.OutputFormat.Stereo.Audio:1", src=CrossfadeTriggerRouteAudio2_0.Out}
  ]
 }]
properties {
 assetImportData=&self:AssetImportData
 numChannels=2
 outputFormat=Stereo
 sampleRate=48000
}
